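# Translucent, two-sided unlit material for the energy effect on the Mouse pickaxe.
# Two panning spike-noise samples combine into one mask that drives a purple-to-cyan color ramp, scaled by an emissive strength parameter.
# The same mask forms opacity, wiped along U by the Reveal parameter through a subtracted smoothstep.
# A sphere mask softens both U ends, with view-angle and depth fades softening it further.
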Material /BRCosmetics/Weapons/FORT_Melee/Pickaxe_Mouse/FX/Materials/M_MousePickaxe_Energy
graph {
 properties {
  bUsedWithInstancedStaticMeshes=true
  blendMode=BLEND_Translucent
  pixelDepthOffsetMode=PDOM_Legacy
  shadingModel=MSM_Unlit
  twoSided=true
  usedShadingModels=Unlit
 }
 nodes [
  n1 = Constant3Vector() {constant=<rgba 0.4548 0.009 1 0>}
  n2 = Constant3Vector() {constant=<rgba 0 0.9787 3 0>}
  n3 = Constant3Vector() {constant=<rgba 0 2.3 5 0>}
  n4 = TextureSample(UVs=Panner(Coordinate=TextureCoordinate() {uTiling=3}) {speedX=-1}) {samplerType=SAMPLERTYPE_LinearColor, texture=/BRCosmetics/Weapons/FORT_Melee/Pickaxe_Mouse/FX/Textures/T_MousePickaxe_Spikenoise_M}
  n5 = TextureSample(UVs=Panner(Coordinate=TextureCoordinate() {uTiling=2.5}) {speedX=-1.25}) {samplerType=SAMPLERTYPE_LinearColor, texture=&T_MousePickaxe_Spikenoise_M}
  n6 = Saturate(None=Multiply(A=Multiply(A=@n4.G) {constB=2}, B=Power(Base=@n5.R) {constExponent=0.4}) {constB=2})
  n7 = SmoothStep(Value=@n6) {constMax=0.75}
  MaterialFunctionCall(Emissive=Multiply(
    A=MaterialFunctionCall(1=@n1, 2=@n2, 3=@n3, A=@n7) {materialFunction=&Lerp_Multiple_Float3}.Lerp 3 Inputs
    B=ScalarParameter() {defaultValue=1, parameterName=EmissiveStrenght}
   )) {materialFunction=&MF_FX_EmissiveAdjustments}
  n8 = LinearInterpolate(B=Constant() {r=1.25}, Alpha=ScalarParameter() {defaultValue=1, parameterName=Reveal})
  n9 = TextureCoordinate()
  n10 = Subtract(
   A=Saturate(None=Power(Base=@n6) {constExponent=1.25})
   B=SmoothStep(Min=Add(A=@n8) {constB=-0.25}, Max=@n8, Value=ComponentMask(None=@n9) {r=true})
  )
  n11 = SphereMask(A=TextureCoordinate() {vTiling=0}, B=Constant2Vector() {r=0.5}, Radius=Constant() {r=0.5}, Hardness=Constant() {r=0.8})
  n12 = Constant() {r=2}
  Multiply(
   A=Multiply(
    A=Multiply(A=@n10, B=@n11)
    B=MaterialFunctionCall("FadeSensitivity (optional)"=@n12) {materialFunction=&FadeBasedOnViewAngle}.Result
   )
   B=DepthFade() {fadeDistanceDefault=10}
  )
 ]
}
