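# Translucent unlit glow material for the Mouse pickaxe weapon FX.
# Two panning inky-smoke samples are summed and shaped by horizontal and vertical sphere-mask falloffs plus a Reveal-driven smoothstep wipe along V.
# That mask drives a three-color purple-to-blue gradient, scaled by EmissiveStrenght and fed through the shared FX emissive adjustment function.
# The same mask is also multiplied by a view-angle fade.
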Material /BRCosmetics/Weapons/FORT_Melee/Pickaxe_Mouse/FX/Materials/M_Mouse_Pickaxe_Glow
graph {
 properties {blendMode=BLEND_Translucent, pixelDepthOffsetMode=PDOM_Legacy, shadingModel=MSM_Unlit, usedShadingModels=Unlit}
 nodes [
  n1 = MaterialFunctionCall() {materialFunction=&WorldPositionBehindTranslucency}
  ComponentMask(None=@n1.Result) {g=true}
  ComponentMask(None=@n1.Result) {r=true}
  n2 = Constant3Vector() {constant=<rgba 0.4548 0.009 1 0>}
  n3 = Constant3Vector() {constant=<rgba 0 0.3262 1 0>}
  n4 = Constant3Vector() {constant=<rgba 0 0.46 1 0>}
  n5 = LinearInterpolate(B=Constant() {r=1.25}, Alpha=ScalarParameter() {defaultValue=1, parameterName=Reveal})
  n6 = Add(A=@n5) {constB=-0.25}
  n7 = ComponentMask(None=TextureCoordinate()) {g=true}
  n8 = ScalarParameter() {defaultValue=0.5}
  n9 = SphereMask(A=TextureCoordinate() {uTiling=0}, B=Constant2Vector() {g=0.5}, Radius=@n8, Hardness=Constant() {r=0.75})
  n10 = SphereMask(A=TextureCoordinate() {vTiling=0}, B=Constant2Vector() {r=0.5}, Radius=@n8, Hardness=Constant() {r=0.5})
  n11 = TextureCoordinate() {uTiling=0.25, vTiling=7}
  n12 = TextureCoordinate() {uTiling=0.15, vTiling=8}
  n13 = Add(
   A=TextureSample(UVs=Panner(Coordinate=@n11) {speedX=-0.2, speedY=1}) {texture=&T_Inky_Smoke_Tile_Chromatic}.R
   B=TextureSample(UVs=Panner(Coordinate=@n12) {speedX=0.2, speedY=1.5}) {texture=&T_Inky_Smoke_Tile_Chromatic}.R
  )
  n14 = Multiply(A=OneMinus(None=SmoothStep(Min=@n6, Max=@n5, Value=@n7)), B=Multiply(A=Multiply(A=@n9, B=@n10), B=Add(A=@n13) {constB=0.5}))
  n15 = Multiply(A=@n14, B=ScalarParameter() {defaultValue=0.33, parameterName=ColorInputStrenght}) {constB=0.25}
  MaterialFunctionCall(Emissive=Multiply(
    A=MaterialFunctionCall(1=@n2, 2=@n3, 3=@n4, A=@n15) {materialFunction=&Lerp_Multiple_Float3}.Lerp 3 Inputs
    B=ScalarParameter() {defaultValue=1, parameterName=EmissiveStrenght}
   )) {materialFunction=&MF_FX_EmissiveAdjustments}
  n16 = Constant() {r=5}
  Multiply(A=Multiply(
    A=@n14
    B=MaterialFunctionCall("FadeSensitivity (optional)"=@n16) {materialFunction=&FadeBasedOnViewAngle}.Result
   ))
 ]
}
