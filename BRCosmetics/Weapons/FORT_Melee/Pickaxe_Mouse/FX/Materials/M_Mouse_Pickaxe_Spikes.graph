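# Unlit, translucent, two-sided FX material for glowing, smoky spikes on the Mouse pickaxe, flagged for Niagara sprites.
# Panning smoke noise, time-offset by a DynamicParameter, combines with UV gradients, a cross-width SphereMask and a Reveal-driven wipe along V.
# That mask drives a purple-blue-cyan emissive ramp and, faded by view angle and depth, an opacity term.
# SplineThicken builds tapered width from base to tip.
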Material /BRCosmetics/Weapons/FORT_Melee/Pickaxe_Mouse/FX/Materials/M_Mouse_Pickaxe_Spikes
graph {
 properties {
  bUsedWithNiagaraSprites=true
  blendMode=BLEND_Translucent
  pixelDepthOffsetMode=PDOM_Legacy
  shadingModel=MSM_Unlit
  twoSided=true
  usedShadingModels=Unlit
 }
 nodes [
  MaterialFunctionCall(WidthBase=Constant() {r=6}, WidthTip=Constant() {r=0.5}) {materialFunction=&SplineThicken}
  n1 = Constant3Vector() {constant=<rgba 0.226 0.0697 0.3542 0>}
  n2 = Constant3Vector() {constant=<rgba 0.0603 0.154 1 0>}
  n3 = Constant3Vector() {constant=<rgba 0 0.92 2 0>}
  n4 = SphereMask(A=TextureCoordinate() {vTiling=0}, B=Constant2Vector() {r=0.5}, Radius=Constant() {r=0.5}, Hardness=Constant())
  n5 = LinearInterpolate(B=Constant() {r=1.25}, Alpha=ScalarParameter() {parameterName=Reveal})
  n6 = ComponentMask(None=TextureCoordinate()) {g=true}
  n7 = Panner(
   Coordinate=TextureCoordinate() {vTiling=0.5}
   Time=Add(A=Time(), B=DynamicParameter() {paramNames [RandomTimeOffset, Param2, Param3, Param4]}.RandomTimeOffset)
  ) {speedX=0.1, speedY=-1}
  n8 = Multiply(
   A=@n6
   B=ComponentMask(None=TextureSample(UVs=@n7) {texture=&T_Inky_Smoke_Tile_Chromatic}.R) {g=true}
  )
  n9 = Multiply(A=Power(Base=@n6)) {constB=0.75}
  n10 = OneMinus(None=@n6)
  n11 = Multiply(A=SmoothStep(Min=Add(A=@n5) {constB=-0.25}, Max=@n5, Value=@n6), B=Multiply(A=Add(A=@n8, B=@n9), B=SmoothStep(Value=@n10) {constMax=0.25}))
  n12 = Saturate(None=Multiply(A=Multiply(A=@n4, B=@n11)) {constB=1.25})
  n13 = SmoothStep(Value=Power(Base=@n12) {constExponent=0.8}) {constMax=0.75}
  MaterialFunctionCall(Emissive=Multiply(
    A=MaterialFunctionCall(1=@n1, 2=@n2, 3=@n3, A=@n13) {materialFunction=&Lerp_Multiple_Float3}.Lerp 3 Inputs
    B=ScalarParameter() {defaultValue=1, parameterName=EmissiveStrenght}
   )) {materialFunction=&MF_FX_EmissiveAdjustments}
  n14 = Constant() {r=1}
  Multiply(
   A=Multiply(
    A=@n12
    B=MaterialFunctionCall("FadeSensitivity (optional)"=@n14) {materialFunction=&FadeBasedOnViewAngle}.Result
   )
   B=DepthFade() {fadeDistanceDefault=10}
  )
 ]
}
