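# Niagara weapon trail system for the Mouse pickaxe, drawing a short-lived energy ribbon that follows the animated skeletal mesh.
# Its single EnergyTrail emitter uses anim-trail functions to spawn, initialize bone indices and update particles against the User skeletal mesh.
# Over particle life, color curves fade a bright blue-cyan glow and drive a separate OutlineColor attribute, while ribbon width tapers to zero.
# Two ribbon renderers draw the trail.
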NiagaraSystem /BRCosmetics/Weapons/FORT_Melee/Pickaxe_Mouse/FX/Systems/NS_Mouse_Pickaxe_Trail
dependencies {
 usedDataInterfaces [&NiagaraDataInterfaceColorCurve, &NiagaraDataInterfaceCurve]
 usedDynamicInputs []
 usedModules [&SystemState]
}
emitters {EnergyTrail {
  emitterSpawnScript [/SetVariables(Emitter.SpherePos:Vector4f=<0 0.75 0 0>) {isSetParameters=true}]
  emitterUpdateScript [
   &EmitterState(Life Cycle Mode=::NewEnumerator1, Loop Behavior=::NewEnumerator1, Loop Duration=0.3, UseLoopDelay=false)
   &FN_SpawnAnimTrail()
  ]
  particleSpawnScript [
   &FN_InitializeAnimTrailBoneIndecies()
   &FN_InitializeAnimTrail(Color=<rgba 1 1 1 0.2>, Lifetime=0.15)
  ]
  particleUpdateScript [
   &FN_UpdateAnimTrail(Locked Particle RibbonWidth=1, SkeletalMesh=$User.SkeletalMesh) {module=FN_UpdateAnimTrail001}
   &ScaleColor(
    Scale RGB=&MakeVectorFromLinearColorRGB(Color=&ColorFromCurve(ColorCurve=DataInterface() {propertyValues {
        AlphaCurve {keys [<at 0 1 RCIM_Cubic>]}
        BlueCurve {keys [<at 0 10>,<at 0.5299 0.151>]}
        GreenCurve {keys [<at 0 1.5586>,<at 0.5299 0.0086>]}
        RedCurve {keys [<at 0 0>,<at 0.5299 0>]}
       }}))
    ScaleA=true
    ScaleRGB=true
   )
   /SetVariables(Particles.OutlineColor:LinearColor=&ColorFromCurve(ColorCurve=DataInterface() {propertyValues {
       AlphaCurve {keys [<at 0 2 RCIM_Cubic>]}
       BlueCurve {keys [<at 0.1705 1 RCIM_Cubic>,<at 0.5578 0.2448 RCIM_Cubic>,<at 1 0.1146>]}
       GreenCurve {keys [<at 0.1705 0.1344 RCIM_Cubic>,<at 0.5578 0.0144 RCIM_Cubic>,<at 1 0>]}
       RedCurve {keys [<at 0.1705 0 RCIM_Cubic>,<at 0.5578 0.2305 RCIM_Cubic>,<at 1 0.0934>]}
      }})) {isSetParameters=true}
   &ScaleRibbonWidth(Ribbon Width Scale=&FloatFromCurve(FloatCurve=DataInterface() {propertyValues {Curve {keys [
         {arriveTangent=-0.0405, interpMode=RCIM_Cubic, leaveTangent=-0.0405, tangentMode=RCTM_User, time=0, value=1}
         {arriveTangent=-2.3676, interpMode=RCIM_Cubic, leaveTangent=-2.3676, tangentMode=RCTM_User, time=1, value=0}
        ]}}}))
  ]
  rendererClasses [&NiagaraRibbonRendererProperties]
  renderers [
   {rendererClass=&NiagaraRibbonRendererProperties}
   {rendererClass=&NiagaraRibbonRendererProperties, rendererIndex=1}
  ]
 }}
summary {
 systemName=NS_Mouse_Pickaxe_Trail
 emitters [{emitterName=EnergyTrail, rendererClasses [&NiagaraRibbonRendererProperties]}]
 userVariables [
  User.bIsFrontEnd {type=NiagaraBool}
  User.bIsFrontEndPreview {type=NiagaraBool}
  User.RibbonIDOffset {type=NiagaraInt32}
  User.RibbonWidth {type=NiagaraFloat, defaultValue=10, description=""}
  User.SkeletalMesh {
   type=NiagaraDataInterfaceSkeletalMesh
   defaultValue {
    struct=NiagaraExt_VariableValue_DataInterface
    value {dataInterface=&self:NiagaraDataInterfaceSkeletalMesh_4}
   }
  }
  User.VariantNum {type=NiagaraFloat, defaultValue=0, description=""}
 ]
}
